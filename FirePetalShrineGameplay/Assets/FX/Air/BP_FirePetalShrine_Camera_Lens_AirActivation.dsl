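; Camera lens effect Blueprint, named for Air activation in FirePetalShrine, that pairs a Niagara component with a post-process effect.
; When the Niagara component activates, a timeline plays from the start and writes its float track into the post-process component's blend weight each update.
; On destruction it explicitly destroys the post-process component, and BeginPlay has no logic.
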(event EventBeginPlay)

(event OnComponentActivated(NiagaraComponent)
  (params (Component :type Actor%20Component%20Object%20Reference) (bReset :type Boolean))
  (bind (_direction _newtrack_0) (Timeline|Timeline PlayFromStart)
    (:Update
      (Variables|/Script/Engine.PostProcessComponent|SetBlendWeight _newtrack_0 (Variables|Default|GetPostProcess)))))

(event EventDestroyed
  (Components|DestroyComponent (Variables|Default|GetPostProcess)))

(fn ConstructionScript
  (|CallParentFunction))
