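# Looping air ambience sound cue, named for the activated state of the Fire Petal Shrine.
# A single looping wave player feeds a Modulator at default settings, which slightly varies pitch and volume each time the cue starts.
# The cue plays at reduced volume in the Fort_Loot_Reward_PSM sound class and references the shrine's activated-ambient attenuation asset for its spatial behavior.
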SoundCue /FirePetalShrineGameplay/Assets/Sounds/Cues/SC_FirePetalShrine_Activated_Air_Ambient_Loop
root=Modulator_0
nodes [
 {id=Modulator_0, children [WavePlayer_0]}
 {
  id=WavePlayer_0
  properties {bLooping=true, soundWaveAssetPtr=/FirePetalShrineGameplay/Assets/Sounds/Waves/FirePetalShrine_Activated_Air_Ambient_Loop}
 }
]
properties {
 attenuationSettings=/FirePetalShrineGameplay/Assets/Sounds/Settings/AS_FirePetalShrine_Activated_Ambient
 maxDistance=5500
 soundClassObject=&Fort_Loot_Reward_PSM
 volumeMultiplier=0.25
 attenuationOverrides {bAttenuate=false, bEnableReverbSend=false, bSpatialize=false}
 concurrencyOverrides {resolutionRule=PreventNew}
}
