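# Sound cue for the air sprite activation at the Fire Petal Shrine.
# After a short fixed delay, it passes through a default Modulator to an equally weighted random pick of three air activation waves, with the Random node excluded from branch culling.
# It uses the shrine's activation attenuation asset and a loot reward sound class.
# Overridden concurrency allows one instance per owner and blocks new plays while one is active.
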SoundCue /FirePetalShrineGameplay/Assets/Sounds/Cues/SC_FirePetalShrine_AirSprite_Activate
root=Delay_0
nodes [
 {id=Delay_0, children [Modulator_0], properties {delayMax=0.2, delayMin=0.2}}
 {id=Modulator_0, children [Random_0]}
 {id=Random_0, children [WavePlayer_1,WavePlayer_2,WavePlayer_3], properties {bShouldExcludeFromBranchCulling=true, weights [1,1,1]}}
 {id=WavePlayer_1, properties {soundWaveAssetPtr=/FirePetalShrineGameplay/Assets/Sounds/Waves/FirePetalShrine_Activated_Air_01}}
 {id=WavePlayer_2, properties {soundWaveAssetPtr=/FirePetalShrineGameplay/Assets/Sounds/Waves/FirePetalShrine_Activated_Air_02}}
 {id=WavePlayer_3, properties {soundWaveAssetPtr=/FirePetalShrineGameplay/Assets/Sounds/Waves/FirePetalShrine_Activated_Air_03}}
]
properties {
 attenuationSettings=/FirePetalShrineGameplay/Assets/Sounds/Settings/AS_FirePetalShrine_Activate
 bOverrideConcurrency=true
 maxDistance=8600
 soundClassObject=&Fort_Loot_Reward_PSM
 volumeMultiplier=0.8
 attenuationOverrides {bAttenuate=false, bEnableReverbSend=false, bSpatialize=false}
 concurrencyOverrides {bLimitToOwner=true, maxCount=1, resolutionRule=PreventNew}
}
